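; Dual-wield Gatehound claw pickaxe cosmetic Blueprint that drives swing and impact Niagara effects.
; Main-hand and offhand swings and gameplay impacts pass the montage section index to their effects as SwingSectionIndex.
; Frontend preview instead deactivates the default impact and spawns one before the pawn, cycling four sections with a DoOnce and delay guarding against duplicate signals.
; Equipping caches the pawn and particle variant index for that preview.
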(event EventSwingRight_Common
  (|CallParentFunction)
  (bind _section_index (Utilities|SetupSwingMontageSectionIndex (Variables|Actor|GetInstigator))
    (:Out
      (Effects|Components|ParticleSystem|SetIntParameter (CallFunction|GetSwingVFXComponent) "SwingSectionIndex" _section_index))))

(event EventSwingLeft_Common
  (|CallParentFunction)
  (bind _section_index (Utilities|SetupSwingMontageSectionIndex (Variables|Actor|GetInstigator))
    (:Out
      (Effects|Components|ParticleSystem|SetIntParameter (CallFunction|GetSwingOffhandVFXComponent) "SwingSectionIndex" _section_index))))

(event EventSwingLeft
  (|CallParentFunction))

(event EventSwingRight
  (|CallParentFunction))

(event FortWeapon|EventOnPlayImpactFX
  (params (HitResult :type Hit%20Result%20Structure%20%28by%20ref%29) (ImpactPhysicalSurface :type EPhysicalSurface%20Enum :default "SurfaceType_Default") (SpawnedPSC :type FXSystem%20Component%20Object%20Reference))
  (bind (_returnvalue_location _returnvalue_rotation _returnvalue_scale) (Math|Transform|BreakTransform (Transformation|GetActorTransform (Variables|Default|GetAs%20Fort%20Player%20Pawn))))
  (|CallParentFunction HitResult ImpactPhysicalSurface SpawnedPSC)
  (if (Reactive|Frontend|IsCosmeticObjectDisplayedasPreview 0)
    ;[0 0 0] frontend
    (Components|Activation|Deactivate SpawnedPSC)
    (bind _returnvalue (Niagara|SpawnSystematLocation "/BRCosmetics/Weapons/FORT_Melee/Pickaxe_GateHound/FX/NS_Pickaxe_Gatehound_Impact.NS_Pickaxe_Gatehound_Impact" (Math|Transform|TransformLocation (Transformation|GetActorTransform (Variables|Default|GetAs%20Fort%20Player%20Pawn)) "100.000000,0.000000,50.000000") (Math|Rotator|CombineRotators _returnvalue_rotation "0.000000,180.000000,0.000000") _returnvalue_scale true false))
    (Utilities|FlowControl|DoOnce false
      (:Completed
        (Effects|Components|ParticleSystem|SetIntParameter _returnvalue "SwingSectionIndex" (Math|Integer|%25%28Integer%29 (Variables|Default|GetFE_Counter) 4))
        (Effects|Components|ParticleSystem|SetBoolParameter _returnvalue "bIsFrontEndPreview" true)
        (Effects|Components|ParticleSystem|SetFloatParameter _returnvalue "Variant" (Math|Conversions|ToFloat(Integer) (Variables|Default|GetFE_VariantIndex)))
        (Components|Activation|Activate _returnvalue false)
        (Math|Integer|IncrementInt (Variables|Default|GetFE_Counter)
          (:%20%20
            (Utilities|FlowControl|RetriggerableDelay 0.25) ; prevents duplicated signal
          ))))
    (else
      (bind _section_index (Utilities|SetupSwingMontageSectionIndex (Variables|Actor|GetInstigator))
        (:Out
          (Effects|Components|ParticleSystem|SetIntParameter (Variables|Default|GetImpact%20FX) "SwingSectionIndex" _section_index))))))

(event FortWeapon|EventOnEquippedAndAttached
  (|CallParentFunction)
  (bind _asfort_player_pawn (Utilities|Casting|CastToFortPlayerPawn (Variables|Actor|GetInstigator)))
  (Variables|Default|SetAs%20Fort%20Player%20Pawn _asfort_player_pawn)
  (bind _index (CosmeticLoadout|GetVariantChannelActiveIndexforSlotActor "(TagName=\"LoadoutSlot.Pickaxe\")" _asfort_player_pawn "(TagName=\"Cosmetics.Variant.Channel.Particle\")" (Utilities|Array|MakeArray "(TagName=\"Cosmetics.Variant.Property.Particle1\")" "(TagName=\"Cosmetics.Variant.Property.Particle2\")")))
  (Variables|Default|SetFE_VariantIndex _index))

(fn ConstructionScript
  (|CallParentFunction))
