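; Dual-wield pickaxe cosmetic Blueprint for AnvilAlarmMoth that animates its idle VFX in frontend and preview displays.
; On weapon cosmetic init it runs the parent logic, then checks for a frontend or preview context.
; A short delay and a timeline drive the "MusicFE" float parameter on the idle VFX component, with another delay when the timeline finishes.
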(event EventOnInitWeaponCosmetics
  (|CallParentFunction)
  (if (or (Reactive|Frontend|IsCosmeticObjectDisplayedinFrontendContext 0) (Reactive|Frontend|IsCosmeticObjectDisplayedasPreview 0))
    (Utilities|FlowControl|Delay 2.5)
    (bind (_direction _newtrack_0) (Timeline|Timeline_0 PlayFromStart)
      (:Update
        _)
      (:Finished
        (Utilities|FlowControl|Delay 2.5)))
    (Effects|Components|ParticleSystem|SetFloatParameter (CallFunction|GetIdleVFXComponent) "MusicFE" _newtrack_0)))

(fn ConstructionScript
  (|CallParentFunction))
