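; Dual-wield CrispSeason pickaxe cosmetic that randomly picks a paint color for the weapon mesh and its effects.
; On non-authority instances, cosmetic init and becoming visible roll an index into PaintColors, set the material's PaintColor, and activate idle VFX.
; The index also feeds idle, trail and impact particles as PaintID.
; A looping timer rerolls the color in frontend display, and hiding the weapon deactivates idle VFX.
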(event FortWeapon|EventOnPlayImpactFX
  (params (HitResult :type Hit%20Result%20Structure%20%28by%20ref%29) (ImpactPhysicalSurface :type EPhysicalSurface%20Enum :default "SurfaceType_Default") (SpawnedPSC :type FXSystem%20Component%20Object%20Reference))
  ;[1 1 1] FX Variables
  (|CallParentFunction HitResult ImpactPhysicalSurface SpawnedPSC)
  (Effects|Components|ParticleSystem|SetIntParameter SpawnedPSC "PaintID" (Variables|Default|GetCurrentID)))

(event FortWeapon|EventOnEquippedAndAttached
  (|CallParentFunction)
  (if (Reactive|Frontend|IsCosmeticObjectDisplayedinFrontendContext 0)
    (Utilities|Time|SetTimerbyEvent (delegate Custom|RetriggerColor) 4.0 true)))

(event EventBeginPlay)

(event Custom|RetriggerColor
  (bind _returnvalue (CallFunction|GetIdleVFXComponent))
  (bind _currentid (Variables|Default|GetCurrentID))
  ;[1 1 1] Random Color for Material
  (Variables|Default|SetCurrentID (Math|Random|RandomIntegerinRange 0 6))
  (if (== (Variables|Default|GetPreviousID) (Variables|Default|GetCurrentID))
    _)
  (Rendering|Material|SetColorParameterValueonMaterials (Variables|Visual|GetWeaponMesh) "PaintColor" (Math|Conversions|ToLinearColor(Color) (Utilities|Array|Get(aref) (Variables|Default|GetPaintColors) (Variables|Default|GetCurrentID))))
  (Variables|Default|SetPreviousID (Variables|Default|GetCurrentID))
  ;[1 1 1] FX Variables
  (Components|Activation|Activate _returnvalue false)
  (Effects|Components|ParticleSystem|SetIntParameter _returnvalue "PaintID" _currentid)
  (Effects|Components|ParticleSystem|SetIntParameter (CallFunction|GetAnimTrailVFXComponent) "PaintID" _currentid))

(event EventOnInitWeaponCosmetics
  (bind _returnvalue (CallFunction|GetIdleVFXComponent))
  (bind _currentid (Variables|Default|GetCurrentID))
  (|CallParentFunction)
  (Networking|SwitchHasAuthority
    (:Remote
      ;[1 1 1] Random Color for Material
      (Variables|Default|SetCurrentID (Math|Random|RandomIntegerinRange 0 6))
      (if (== (Variables|Default|GetPreviousID) (Variables|Default|GetCurrentID))
        _)
      (Rendering|Material|SetColorParameterValueonMaterials (Variables|Visual|GetWeaponMesh) "PaintColor" (Math|Conversions|ToLinearColor(Color) (Utilities|Array|Get(aref) (Variables|Default|GetPaintColors) (Variables|Default|GetCurrentID))))
      (Variables|Default|SetPreviousID (Variables|Default|GetCurrentID))
      ;[1 1 1] FX Variables
      (Components|Activation|Activate _returnvalue false)
      (Effects|Components|ParticleSystem|SetIntParameter _returnvalue "PaintID" _currentid)
      (Effects|Components|ParticleSystem|SetIntParameter (CallFunction|GetAnimTrailVFXComponent) "PaintID" _currentid))))

(event EventOnWeaponVisibilityChanged
  (params (bVisible :type Boolean) (bSetForLocalControllerOnly :type Boolean))
  (bind _returnvalue (CallFunction|GetIdleVFXComponent))
  (bind _currentid (Variables|Default|GetCurrentID))
  (if bVisible
    (Networking|SwitchHasAuthority
      (:Remote
        ;[1 1 1] Random Color for Material
        (Variables|Default|SetCurrentID (Math|Random|RandomIntegerinRange 0 6))
        (if (== (Variables|Default|GetPreviousID) (Variables|Default|GetCurrentID))
          _)
        (Rendering|Material|SetColorParameterValueonMaterials (Variables|Visual|GetWeaponMesh) "PaintColor" (Math|Conversions|ToLinearColor(Color) (Utilities|Array|Get(aref) (Variables|Default|GetPaintColors) (Variables|Default|GetCurrentID))))
        (Variables|Default|SetPreviousID (Variables|Default|GetCurrentID))
        ;[1 1 1] FX Variables
        (Components|Activation|Activate _returnvalue false)
        (Effects|Components|ParticleSystem|SetIntParameter _returnvalue "PaintID" _currentid)
        (Effects|Components|ParticleSystem|SetIntParameter (CallFunction|GetAnimTrailVFXComponent) "PaintID" _currentid)))
    (else
      (Components|Activation|Deactivate (CallFunction|GetIdleVFXComponent)))))

(fn ConstructionScript
  (|CallParentFunction))
