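; Dual-wield pickaxe Blueprint for the CubeCoast cosmetic that resets its idle VFX after each left or right swing ends.
; Each swing-end event takes the main or offhand idle VFX component and, if valid, waits through a retriggerable delay before setting its "bisSwinging" parameter false.
; The swing VFX validity checks and the trailing second delay lead to no further actions, and the construction script only calls the parent.
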(event EventSwingLeftEnd
  (bind _returnvalue (CallFunction|GetIdleVFXComponent))
  (Utilities|IsValid (CallFunction|GetSwingVFXComponent)
    (:Is%20Valid
      _)
    (:Is%20Not%20Valid
      _))
  (Utilities|IsValid _returnvalue
    (:Is%20Valid
      (Utilities|FlowControl|RetriggerableDelay 0.75)
      (Effects|Components|ParticleSystem|SetBoolParameter _returnvalue "bisSwinging" false)
      (Utilities|FlowControl|RetriggerableDelay 0.45))))

(event EventSwingRightEnd
  (bind _returnvalue (CallFunction|GetIdleOffhandVFXComponent))
  (Utilities|IsValid (CallFunction|GetSwingOffhandVFXComponent)
    (:Is%20Valid
      _)
    (:Is%20Not%20Valid
      _))
  (Utilities|IsValid _returnvalue
    (:Is%20Valid
      (Utilities|FlowControl|RetriggerableDelay 0.75)
      (Effects|Components|ParticleSystem|SetBoolParameter _returnvalue "bisSwinging" false)
      (Utilities|FlowControl|RetriggerableDelay 0.45))))

(fn ConstructionScript
  (|CallParentFunction))
