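; Blueprint for the DenseFog dual-wield pickaxe cosmetic, adding swing-hand awareness to its impact effects.
; Left and right swing events call their parents, then record the hand in the SwingLeft variable.
; On impact, the parent handler runs, then the SwingLeft flag is copied to the spawned FX component's matching bool parameter, letting the effect differ per hand.
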(event EventBeginPlay)

(event EventTick
  (params (DeltaSeconds :type Float%20%28single-precision%29)))

(event FortWeapon|EventOnPlayImpactFX
  (params (HitResult :type Hit%20Result%20Structure%20%28by%20ref%29) (ImpactPhysicalSurface :type EPhysicalSurface%20Enum :default "SurfaceType_Default") (SpawnedPSC :type FXSystem%20Component%20Object%20Reference))
  (|CallParentFunction HitResult ImpactPhysicalSurface SpawnedPSC)
  (if (Variables|Default|GetSwingLeft)
    (Effects|Components|ParticleSystem|SetBoolParameter SpawnedPSC "SwingLeft" true)
    (else
      (Effects|Components|ParticleSystem|SetBoolParameter SpawnedPSC "SwingLeft" false))))

(event EventSwingRight_Common
  (|CallParentFunction)
  (Variables|Default|SetSwingLeft false))

(event EventSwingLeft_Common
  (|CallParentFunction)
  (Variables|Default|SetSwingLeft true))

(fn ConstructionScript
  (|CallParentFunction))
