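; Dual-wield fists pickaxe Blueprint that customizes the melee swing VFX inherited from its parent.
; On the second left swing, it runs the parent logic, then reads the Niagara system from the swing VFX component and spawns another instance attached to the weapon mesh.
; It sets SwingSectionIndex on that instance to select that swing's section.
; BeginPlay and Tick are empty, and construction only calls the parent.
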(event EventBeginPlay)

(event EventTick
  (params (DeltaSeconds :type Float%20%28single-precision%29)))

(event EventSwingLeft2
  (|CallParentFunction)
  (bind _asniagara_system (Utilities|Casting|CastToNiagaraSystem (Effects|Components|ParticleSystem|GetFXSystemAsset (CallFunction|GetSwingVFXComponent))))
  (bind _returnvalue (Niagara|SpawnSystemAttached _asniagara_system (Variables|Visual|GetWeaponMesh) "None" "0, 0, 0" "0, 0, 0" "KeepRelativeOffset" true))
  (Niagara|SetNiagaraVariable(Int32) _returnvalue "SwingSectionIndex" 3))

(fn ConstructionScript
  (|CallParentFunction))
