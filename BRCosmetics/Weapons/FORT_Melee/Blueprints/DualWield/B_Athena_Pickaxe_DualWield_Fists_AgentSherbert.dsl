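; Dual-wield fists pickaxe Blueprint for the AgentSherbert cosmetic, customizing Niagara swing-trail and impact effects.
; Swing events record hand side and whether it is the final Left2 swing, which spawns its own trail.
; Impact FX align to the hit normal with a per-hand rotation; the final swing attaches a second rotated impact and drives a SecondImpact parameter.
; SetParamsOnTrails is empty here and GetFXParamNames returns None.
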(event EventSwingLeft2
  (|CallParentFunction)
  (Variables|Default|SetLeft2%20Swing? true)
  ;[1 1 1] Final Swing
  (bind _returnvalue (Niagara|SpawnSystemAttached "/BRCosmetics/Weapons/FORT_Melee/Pickaxe_AgentSherbert/FX/NS_Pickaxe_AgentSherbert_Trail.NS_Pickaxe_AgentSherbert_Trail" (Variables|Visual|GetWeaponMesh) "None" "0, 0, 0" "0, 0, 0" "KeepRelativeOffset" true))
  (Niagara|SetNiagaraVariable(Float) _returnvalue "AnimTrailWidth" 1.7)
  (Niagara|SetNiagaraVariable(Float) _returnvalue (CallFunction|GetFXParamNames) 0.25)
  (Niagara|SetNiagaraVariable(Float) _returnvalue "Lifetime" 0.175))

(event FortWeapon|EventOnPlayImpactFX
  (params (HitResult :type Hit%20Result%20Structure%20%28by%20ref%29) (ImpactPhysicalSurface :type EPhysicalSurface%20Enum :default "SurfaceType_Default") (SpawnedPSC :type FXSystem%20Component%20Object%20Reference))
  (bind (_bblockinghit _binitialoverlap _time _distance _location _impactpoint _normal) (Collision|BreakHitResult HitResult))
  (bind _returnvalue (Math|Rotator|MakeRotfromX _normal))
  (|CallParentFunction HitResult ImpactPhysicalSurface SpawnedPSC)
  (bind _asniagara_particle_system_component (Utilities|Casting|CastToNiagaraComponent SpawnedPSC))
  (Utilities|FlowControl|Sequence
    (:then_0
      (Variables|Default|SetImpact%20FX_0 _asniagara_particle_system_component)
      (Transformation|SetWorldRotation SpawnedPSC (Math|Rotator|CombineRotators _returnvalue (select (Variables|Default|GetRight?) (Variables|Default|GetRight%20Rotation) (Variables|Default|GetLeft%20Rotation))) false false)
      (if (Variables|Default|GetLeft2%20Swing?)
        (bind _returnvalue_2 (Niagara|SpawnSystemAttached "/BRCosmetics/Weapons/FORT_Melee/Pickaxe_AgentSherbert/FX/NS_Pickaxe_AgentSherbert_Impact.NS_Pickaxe_AgentSherbert_Impact" (Variables|Default|GetImpact%20FX_0) "None" "0, 0, 0" "0, 0, 0" "KeepRelativeOffset" true))
        (Transformation|SetWorldRotation _returnvalue_2 (Math|Rotator|CombineRotators _returnvalue "0,60.000000, 0") false false)))
    (:then_1
      (Niagara|SetNiagaraVariable(Bool) _asniagara_particle_system_component "SecondImpact" (Variables|Default|GetLeft2%20Swing?)))))

(event EventSwingLeft_Common
  (|CallParentFunction)
  (Variables|Default|SetRight? false)
  (CallFunction|SetParamsonTrails 0.15 0.18))

(event EventSwingRight_Common
  (|CallParentFunction)
  (Variables|Default|SetRight? true)
  (CallFunction|SetParamsonTrails 0.15 0.18))

(event EventSwingLeft
  (|CallParentFunction)
  (Variables|Default|SetLeft2%20Swing? false))

(event EventSwingRight
  (|CallParentFunction)
  (Variables|Default|SetLeft2%20Swing? false))

(fn GetFXParamNames
  (return "None"))

(fn SetParamsOnTrails
  (params (LoopDuration :type Float%20%28single-precision%29) (Lifetime :type Float%20%28single-precision%29)))

(fn ConstructionScript
  (|CallParentFunction))
