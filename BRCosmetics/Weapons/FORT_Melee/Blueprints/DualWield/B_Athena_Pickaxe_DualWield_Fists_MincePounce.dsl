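; Dual-wield fists pickaxe Blueprint for the MincePounce cosmetic, adding swing-specific trail effects.
; Each right or left swing runs the parent logic, then gets the swing montage section index from the instigator.
; That index is set as the SwingSectionIndex parameter on the main-hand or off-hand trail particle component, respectively.
; SetParamsOnTrails is left empty and GetFXParamNames returns None, so neither supplies trail parameters.
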(event EventSwingRight_Common
  ;[0 0 0] Swing
  (|CallParentFunction)
  (bind _section_index (Utilities|SetupSwingMontageSectionIndex (Variables|Actor|GetInstigator))
    (:Out
      (Effects|Components|ParticleSystem|SetIntParameter (CallFunction|GetAnimTrailVFXComponent) "SwingSectionIndex" _section_index))))

(event EventSwingLeft_Common
  ;[0 0 0] Swing
  (|CallParentFunction)
  (bind _section_index (Utilities|SetupSwingMontageSectionIndex (Variables|Actor|GetInstigator))
    (:Out
      (Effects|Components|ParticleSystem|SetIntParameter (CallFunction|GetAnimTrailsOffhandPSC) "SwingSectionIndex" _section_index))))

(fn GetFXParamNames
  (return "None"))

(fn SetParamsOnTrails
  (params (LoopDuration :type Float%20%28single-precision%29) (Lifetime :type Float%20%28single-precision%29)))

(fn ConstructionScript
  (|CallParentFunction))
