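; Dual-wield fists pickaxe Blueprint for the StunningMountain cosmetic, driving its idle, impact and swing VFX.
; BeginPlay (in frontend display, when the FortPawn cast fails) and visibility changes use delayed, counted checks on the idle VFX component to set bIsOffhand.
; Swing events increment or reset SwingCount, which impact effects receive while being rotated to the hit normal.
; SwingLeft2 attaches a Niagara trail to the weapon mesh and prints leftover debug text.
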(event EventTick
  (params (DeltaSeconds :type Float%20%28single-precision%29)))

(event EventBeginPlay
  (bind _returnvalue (CallFunction|GetIdleVFXComponent))
  (|CallParentFunction)
  (Utilities|FlowControl|DelayUntilNextFrame)
  (Utilities|FlowControl|DelayUntilNextFrame)
  (Utilities|Casting|CastToFortPawn (Game|GetInstigator)
    (:CastFailed
      (if (Reactive|Frontend|IsCosmeticObjectDisplayedinFrontendContext 0)
        (Variables|Default|SetLoopCheckCount 0)
        (Utilities|IsValid _returnvalue
          (:Is%20Valid
            _)
          (:Is%20Not%20Valid
            (Utilities|FlowControl|Delay 0.1)
            (bind _result (Math|Integer|IncrementInt (Variables|Default|GetLoopCheckCount))
              (:%20%20
                (if (<= _result 20)
                  _)))))
        (Effects|Components|ParticleSystem|SetBoolParameter _returnvalue "bIsOffhand" true)))))

(event EventOnWeaponVisibilityChanged
  (params (bVisible :type Boolean) (bSetForLocalControllerOnly :type Boolean))
  (bind _returnvalue (CallFunction|GetIdleVFXComponent))
  (|CallParentFunction bVisible false)
  (if bVisible
    (Utilities|IsValid _returnvalue
      (:Is%20Valid
        _)
      (:Is%20Not%20Valid
        (Utilities|FlowControl|Delay 0.1)
        (bind _result (Math|Integer|IncrementInt (Variables|Default|GetCounter))
          (:%20%20
            (if (>= _result 50)
              _
              (else
                _))
            (Development|PrintString "-----------------------StunningMountainPickaxeBrok-------------------")))))
    (Effects|Components|ParticleSystem|SetBoolParameter _returnvalue "bIsOffhand" true)
    (else
      (Variables|Default|SetCounter 0))))

(event FortWeapon|EventOnPlayImpactFX
  (params (HitResult :type Hit%20Result%20Structure%20%28by%20ref%29) (ImpactPhysicalSurface :type EPhysicalSurface%20Enum :default "SurfaceType_Default") (SpawnedPSC :type FXSystem%20Component%20Object%20Reference))
  (bind (_bblockinghit _binitialoverlap _time _distance _location _impactpoint _normal) (Collision|BreakHitResult HitResult))
  (|CallParentFunction HitResult ImpactPhysicalSurface SpawnedPSC)
  (bind _asniagara_particle_system_component (Utilities|Casting|CastToNiagaraComponent SpawnedPSC))
  (Utilities|FlowControl|Sequence
    (:then_0
      (bind _output_get (Variables|Default|SetImpact%20FX_0 _asniagara_particle_system_component))
      (Transformation|SetWorldRotation SpawnedPSC (Math|Rotator|MakeRotfromX _normal) false false)
      (Effects|Components|ParticleSystem|SetIntParameter _output_get "SwingCount" (Variables|Default|GetSwingCount)))))

(event EventSwingLeft_Common
  (|CallParentFunction)
  (Math|Integer|IncrementInt (Variables|Default|GetSwingCount)))

(event EventSwingRight_Common
  (|CallParentFunction)
  (Math|Integer|IncrementInt (Variables|Default|GetSwingCount)))

(event EventSwingLeft2
  (|CallParentFunction)
  (Development|PrintString "FinalSwing" true true "(R=1.000000,G=0.137375,B=0.000000,A=1.000000)")
  (Niagara|SpawnSystemAttached "/BRCosmetics/Weapons/FORT_Melee/Pickaxe_StunningMountain/FX/NS_StunningMountain_Trail.NS_StunningMountain_Trail" (Variables|Visual|GetWeaponMesh) "None" "0, 0, 0" "0, 0, 0" "KeepRelativeOffset" true))

(event EventSwingRight
  (|CallParentFunction)
  (Variables|Default|SetSwingCount 0))

(fn ConstructionScript
  (|CallParentFunction))
