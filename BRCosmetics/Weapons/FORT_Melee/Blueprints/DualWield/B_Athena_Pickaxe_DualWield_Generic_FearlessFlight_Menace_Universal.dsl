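; Dual-wield pickaxe Blueprint for the FearlessFlight Menace cosmetic that orients its impact effect to the current swing.
; On impact, it calls the parent, then finds the instigator pawn's current montage section in the swing section list.
; That index selects one of four swing vectors, which sets the "ImpactDIrection" parameter on the Impact FX component.
; Debug prints output the section name and chosen vector; BeginPlay and Tick are empty.
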(event EventBeginPlay)

(event EventTick
  (params (DeltaSeconds :type Float%20%28single-precision%29)))

(event FortWeapon|EventOnPlayImpactFX
  (params (HitResult :type Hit%20Result%20Structure%20%28by%20ref%29) (ImpactPhysicalSurface :type EPhysicalSurface%20Enum :default "SurfaceType_Default") (SpawnedPSC :type FXSystem%20Component%20Object%20Reference))
  (bind _montagereference (Variables|Default|GetMontageReference))
  (bind _swing_3 (Variables|Default|GetSwing%203))
  (bind _swing_1 (Variables|Default|GetSwing%201))
  (bind _swing_2 (Variables|Default|GetSwing%202))
  (bind _swing_4 (Variables|Default|GetSwing%204))
  (bind _montage_swing_sections (Variables|Default|GetMontage_Swing_Sections))
  (|CallParentFunction HitResult ImpactPhysicalSurface SpawnedPSC)
  (bind _asfort_pawn (Utilities|Casting|CastToFortPawn (Game|GetInstigator)))
  (Development|PrintString (Utilities|String|ToString(Name) (Animation|Montage|MontageGetCurrentSection (Components|SkeletalMesh|GetAnimInstance (Variables|/Script/Engine.Character|GetMesh _asfort_pawn)) _montagereference)))
  (Effects|Components|ParticleSystem|SetVectorParameter (Variables|Default|GetImpact%20FX) "ImpactDIrection" (Utilities|Select :Index (Utilities|Array|FindItem _montage_swing_sections (Animation|Montage|MontageGetCurrentSection (Components|SkeletalMesh|GetAnimInstance (Variables|/Script/Engine.Character|GetMesh _asfort_pawn)) _montagereference)) :Option%200 _swing_1 :Option%201 _swing_2 :Option%202 _swing_3 :Option%203 _swing_4))
  (Development|PrintString (Utilities|String|ToString(Vector) (Utilities|Select :Index (Utilities|Array|FindItem _montage_swing_sections (Animation|Montage|MontageGetCurrentSection (Components|SkeletalMesh|GetAnimInstance (Variables|/Script/Engine.Character|GetMesh _asfort_pawn)) _montagereference)) :Option%200 _swing_1 :Option%201 _swing_2 :Option%202 _swing_3 :Option%203 _swing_4))))

(fn ConstructionScript
  (|CallParentFunction))
