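; Dual-wield OuterGarment pickaxe cosmetic that glows reactively after qualifying hits on its reactive variant, recognized by its impact Niagara asset.
; Impact FX triggers reactivity in frontend preview or, through ReplicateReactivity, on three weakpoint surface types; other hits pass the current state to the spawned effect.
; A retriggerable delay holds the state between timelines that ramp mesh emissive and idle VFX color blend, then update the impact effect's ReactivityOn.
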(event FortWeapon|EventOnEquippedAndAttached
  ;[1 1 1] Sets reactivity once qeuipped in game - disabled since it does not work in the front end
  (|CallParentFunction))

(event Custom|SetReactivity)

(event Custom|TriggerReactivity
  (CallFunction|EnableReactivity)
  (Utilities|FlowControl|RetriggerableDelay (Variables|Default|GetReactivityDuration))
  (CallFunction|DisableReactivity))

(event Custom|FrontendReactivity
  (CallFunction|EnableReactivity)
  (Utilities|FlowControl|RetriggerableDelay (Variables|Default|GetReactivityDuration))
  (CallFunction|DisableReactivity))

(event FortWeapon|EventOnPlayImpactFX
  (params (HitResult :type Hit%20Result%20Structure%20%28by%20ref%29) (ImpactPhysicalSurface :type EPhysicalSurface%20Enum :default "SurfaceType_Default") (SpawnedPSC :type FXSystem%20Component%20Object%20Reference))
  ;[1 1 1] Gets the weakpoints and sets makes this the condition for the event send.
  (|CallParentFunction HitResult ImpactPhysicalSurface SpawnedPSC)
  (bind _output_get (Variables|Default|SetIs%20Reactive%20Variant (== (Effects|Components|ParticleSystem|GetFXSystemAsset SpawnedPSC) "/BRCosmetics/Weapons/FORT_Melee/Pickaxe_OuterGarment/FX/NS_Pickaxe_OuterGarment_Impact.NS_Pickaxe_OuterGarment_Impact")))
  (if _output_get
    (if (Reactive|Frontend|IsCosmeticObjectDisplayedasPreview 0)
      (CallFunction|FrontendReactivity)
      (Effects|Components|ParticleSystem|SetBoolParameter SpawnedPSC "ReactivityOn" true)
      (else
        (if (Math|Boolean|ORBoolean (Utilities|Enum|Equal(Enum) (Collision|GetSurfaceType HitResult) "SurfaceType14") (Utilities|Enum|Equal(Enum) (Collision|GetSurfaceType HitResult) "SurfaceType15") (Utilities|Enum|Equal(Enum) (Collision|GetSurfaceType HitResult) "SurfaceType16"))
          ;[1 1 1] Replication
          (CallFunction|ReplicateReactivity)
          (else
            ;[1 1 1] Initial bool state is set here. Bool needs to be kept at true during the event chain.
            (Effects|Components|ParticleSystem|SetBoolParameter SpawnedPSC "ReactivityOn" (Variables|Default|GetReactivityOn))))))))

(event Custom|ReplicateReactivity
  ;[1 1 1] Replication
  (CallFunction|TriggerReactivity))

(event Custom|EnableReactivity
  (if (not (Variables|Default|GetReactivityOn))
    (bind _output_get (Variables|Default|SetReactivityOn true))
    (bind (_direction _emissivemultiplier) (Timeline|RampEmissiveUp PlayFromStart)
      (:Update
        (Effects|Components|ParticleSystem|SetFloatParameter (CallFunction|GetIdleVFXComponent) "ColorBlend" _emissivemultiplier)
        (bind _output_get_2 (Variables|Default|SetEmissiveIntensity (* _emissivemultiplier (Variables|Default|GetEmissiveIntensityMax))))
        (Rendering|Material|SetScalarParameterValue (Material|GetMID (Variables|Visual|GetWeaponMesh)) "emissive mult" _output_get_2 :!class /Script/Engine.MaterialInstanceDynamic))
      (:Finished
        (Effects|Components|ParticleSystem|SetBoolParameter (stale FortWeapon|EventOnPlayImpactFX SpawnedPSC) "ReactivityOn" _output_get)))))

(event Custom|DisableReactivity
  (bind _output_get (Variables|Default|SetReactivityOn false))
  (bind (_direction _emissivemultiplier) (Timeline|RampEmissiveDown PlayFromStart)
    (:Update
      (Effects|Components|ParticleSystem|SetFloatParameter (CallFunction|GetIdleVFXComponent) "ColorBlend" _emissivemultiplier)
      (bind _output_get_2 (Variables|Default|SetEmissiveIntensity (* _emissivemultiplier (Variables|Default|GetEmissiveIntensityMax))))
      (Rendering|Material|SetScalarParameterValue (Material|GetMID (Variables|Visual|GetWeaponMesh)) "emissive mult" _output_get_2 :!class /Script/Engine.MaterialInstanceDynamic))
    (:Finished
      (Effects|Components|ParticleSystem|SetBoolParameter (stale FortWeapon|EventOnPlayImpactFX SpawnedPSC) "ReactivityOn" _output_get))))

(event EventOnInitWeaponCosmetics
  (|CallParentFunction)
  (Rendering|Material|SetScalarParameterValue (Material|GetMID (Variables|Visual|GetWeaponMesh)) "emissive mult" 0.0 :!class /Script/Engine.MaterialInstanceDynamic))

(fn ConstructionScript
  (|CallParentFunction))
