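; Dual-wield pickaxe Blueprint for the StairGummy melee cosmetic, adding a custom Niagara trail to one of its swings.
; On the second left swing it runs parent logic, then spawns the StairGummy trail system attached to the weapon mesh.
; It sets the trail's SwingSectionIndex integer to select the section for that swing.
; BeginPlay and Tick are empty, and the construction script only calls the parent.
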(event EventBeginPlay)

(event EventTick
  (params (DeltaSeconds :type Float%20%28single-precision%29)))

(event EventSwingLeft2
  (|CallParentFunction)
  (bind _returnvalue (Niagara|SpawnSystemAttached "/BRCosmetics/Weapons/FORT_Melee/Pickaxe_StairGummy/FX/NS_Pickaxe_StairGummy_Trail.NS_Pickaxe_StairGummy_Trail" (Variables|Visual|GetWeaponMesh) "None" "0, 0, 0" "0, 0, 0" "KeepRelativeOffset" true))
  (Niagara|SetNiagaraVariable(Int32) _returnvalue "SwingSectionIndex" 3))

(fn ConstructionScript
  (|CallParentFunction))
