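; Weapon Blueprint for the Titanium dual-wield pickaxe cosmetic, handling its effects when weapon visibility changes.
; The visibility event toggles the main and offhand alteration ambient particle components, each guarded by a validity check.
; On becoming visible it also activates the idle VFX and, if time-of-day control is enabled, calls TODCheck with current and previous day phases.
; BeginPlay and Tick are empty, and construction only calls the parent.
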(event EventBeginPlay)

(event EventTick
  (params (DeltaSeconds :type Float%20%28single-precision%29)))

(event EventOnWeaponVisibilityChanged
  (params (bVisible :type Boolean) (bSetForLocalControllerOnly :type Boolean))
  (bind _alteration_ambient_ps (Variables|Default|GetAlteration%20Ambient%20PS))
  (bind _offhand_alteration_ambient_ps (Variables|Default|GetOffhand%20Alteration%20Ambient%20PS))
  (Utilities|FlowControl|Sequence
    (:then_0
      (Utilities|IsValid (Variables|Default|GetAlteration%20Ambient%20PS)
        (:Is%20Valid
          (if bVisible
            (Components|Activation|Activate _alteration_ambient_ps true)
            (else
              (Components|Activation|Deactivate _alteration_ambient_ps))))))
    (:then_1
      (Utilities|IsValid (Variables|Default|GetOffhand%20Alteration%20Ambient%20PS)
        (:Is%20Valid
          (if bVisible
            (Components|Activation|Activate _offhand_alteration_ambient_ps true)
            (else
              (Components|Activation|Deactivate _offhand_alteration_ambient_ps))))))
    (:then_2
      (if bVisible
        (Components|Activation|Activate (CallFunction|GetIdleVFXComponent) true)
        (if (Variables|TimeofDayControl|GetUseTimeofDayControl)
          (CallFunction|TODCheck (TimeOfDay|GetCurrentDayPhase (TimeOfDay|GetContextualTimeOfDayManager) :!class /Script/FortniteGame.FortTimeOfDayManagerInterface) (TimeOfDay|GetPreviousDayPhase (TimeOfDay|GetContextualTimeOfDayManager) :!class /Script/FortniteGame.FortTimeOfDayManager)))))))

(fn ConstructionScript
  (|CallParentFunction))
