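; Dual-wield PinkSpike one-handed pickaxe weapon Blueprint that adds an outline overlay to its mesh.
; On equip it activates the idle particle effect, picks an outline material by material quality level and limits the overlay's draw distance.
; UpdateOverlayTextureOnVariant swaps the dynamic material's diffuse textures when the instigator's pickaxe Particle variant index is nonzero.
; The overlay is cleared when the equipped weapon is destroyed.
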(event FortWeapon|EventOnEquippedAndAttached
  (bind _weaponmesh (Variables|Visual|GetWeaponMesh))
  (|CallParentFunction)
  (Components|Activation|Activate (CallFunction|GetIdlePSC) false)
  (Utilities|FlowControl|Sequence
    (:then_1
      ;[1 0 0.3634] Overlay Material
      (switch int (Rendering|Material|GetRenderingMaterialQualityLevel)
        (:Default
          (bind _returnvalue (Rendering|Material|CreateDynamicMaterialInstance "/BRCosmetics/Weapons/FORT_Melee/Pickaxe_PinkSpike/Materials/MI_PinkSpike_Outline_P.MI_PinkSpike_Outline_P" "None" "None" :!class /Script/Engine.KismetMaterialLibrary))
          (CallFunction|UpdateOverlayTextureonVariant _returnvalue)
          (Rendering|Material|SetOverlayMaterial _weaponmesh _returnvalue false 0 :!class /Script/Engine.MeshComponent) ; Original Outlines
        )
        (:0
          (bind _returnvalue_2 (Rendering|Material|CreateDynamicMaterialInstance "/BRCosmetics/Weapons/FORT_Melee/Pickaxe_PinkSpike/Materials/MI_PinkSpike_OverlayOutline_P.MI_PinkSpike_OverlayOutline_P" "None" "None" :!class /Script/Engine.KismetMaterialLibrary))
          (CallFunction|UpdateOverlayTextureonVariant _returnvalue_2)
          (Rendering|Material|SetOverlayMaterial _weaponmesh _returnvalue_2 false 0 :!class /Script/Engine.MeshComponent) ; Black OverlayOutline
        )
        (:1
          _)
        (:2
          _))
      (Rendering|Material|SetOverlayMaterialMaxDrawDistance _weaponmesh 2.5e+03 :!class /Script/Engine.MeshComponent))))

(event EventBeginPlay
  (|CallParentFunction))

(event EventTick
  (params (DeltaSeconds :type Float%20%28single-precision%29)))

(event EventOnEquippedWeaponDestory
  ;[1 0 0.3634] Overlay Material
  (Rendering|Material|SetOverlayMaterial (Variables|Visual|GetWeaponMesh) 0 false 0 :!class /Script/Engine.MeshComponent))

(fn UpdateOverlayTextureOnVariant
  (params (Material%20Ref :type Material%20Instance%20Dynamic%20Object%20Reference%20%28by%20ref%29))
  (bind _dfcolortextureoverride (LocalVariables|GetDFColorTextureOverride))
  (bind _asplayer_pawn_athena (Utilities|Casting|CastToPlayerPawn_Athena (Game|GetInstigator)))
  (bind _index (CosmeticLoadout|GetVariantChannelActiveIndexforSlotActor "(TagName=\"LoadoutSlot.Pickaxe\")" _asplayer_pawn_athena "(TagName=\"Cosmetics.Variant.Channel.Particle\")" (Utilities|Array|MakeArray "(TagName=\"Cosmetics.Variant.Property.Particle1\")" "(TagName=\"Cosmetics.Variant.Property.Particle2\")")))
  (if (Math|Conversions|ToBoolean(Integer) _index)
    (Rendering|Material|SetTextureParameterValue Material%20Ref "Diffuse" _dfcolortextureoverride)
    (Rendering|Material|SetTextureParameterValue Material%20Ref "OutlineDiffuseTexture" _dfcolortextureoverride)
    (else
      (return))))

(fn ConstructionScript
  (|CallParentFunction))
