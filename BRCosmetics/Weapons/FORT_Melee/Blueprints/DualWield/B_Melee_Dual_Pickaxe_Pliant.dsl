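; Dual-wield melee pickaxe Blueprint for the Pliant cosmetic, which configures weapon mesh materials when equipped.
; On equip, it sets a position-switch material scalar on the offhand mesh after a zero-second delay that avoids a one-frame-late update, identically in both branches of a frontend-display check.
; In frontend display only, it also writes custom primitive data on the main mesh to identify it as the female pickaxe for wrap materials.
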(event EventTick
  (params (DeltaSeconds :type Float%20%28single-precision%29)))

(event FortWeapon|EventOnEquippedAndAttached
  (bind _self self)
  (bind _returnvalue (Reactive|Frontend|IsCosmeticObjectDisplayedinFrontendContext _self))
  (bind _weaponmeshoffhand (Variables|Visual|GetWeaponMeshOffhand))
  (|CallParentFunction)
  (Utilities|FlowControl|Sequence
    (:then_0
      (if _returnvalue
        (Utilities|FlowControl|Delay 0.0) ; Without 0 delay, the position updates a frame late
        (Rendering|Material|SetScalarParameterValueonMaterials _weaponmeshoffhand "DO NOT OVERRIDE-Pickaxe-Position-BP-Switch" 1.0)
        (else
          (Utilities|FlowControl|Delay 0.0) ; Without 0 delay, the position updates a frame late
          (Rendering|Material|SetScalarParameterValueonMaterials _weaponmeshoffhand "DO NOT OVERRIDE-Pickaxe-Position-BP-Switch" 1.0))))
    (:then_1
      (if _returnvalue
        ;[1 1 1] Pliant Specific Custom Data for Wrap Identification
        (Rendering|Material|SetCustomPrimitiveDataFloat (Variables|Visual|GetWeaponMesh) 0 4.0) ; 0 is default (I.E. anything else using the wrap, 1 is Pliant Body, 2 is Pliant Backpack, 3 is Pliant Male Pickaxe 4 is Female Pickaxe
      ))))

(fn ConstructionScript
  (|CallParentFunction))
